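# Niagara effect for the FoundKelp emote: blue electrical energy on the character's right hand that culminates in a timed shockwave boom.
# Lightning coil meshes, sparks, layered glow and a ribbon trail sample the hand_r bone, with spawn rates and brightness shaped by curves.
# A shockwave emitter bursts about three seconds in, using spawn groups to drive shockwave, airflow, flash, refraction and particle layers.
# Colors and intensities adjust for time of day and front-end display.
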NiagaraSystem /BRCosmetics/Animation/Game/MainPlayer/Emotes/FoundKelp/FX/NS_Emote_FoundKelp_Boom
dependencies {
 usedDataInterfaces [&NiagaraDataInterfaceCurve, &NiagaraDataInterfaceSkeletalMesh, &NiagaraDataInterfaceColorCurve]
 usedDynamicInputs [&RandomRangeVector]
 usedModules [
  &self:SystemSpawnScript.SystemScriptSource.SystemScriptGraph.NiagaraNodeAssignment_0.SetVariables_#1
  &SystemState
 ]
}
emitters {
 ElectricCoils {
  emitterUpdateScript [
   &EmitterState()
   &SpawnRate(
    Spawn Probability=0.525
    SpawnRate=&FloatFromCurve(
     FloatCurve=DataInterface() {propertyValues {Curve {keys [
         <at 0.2299 1.3383 RCIM_Cubic>
         {arriveTangent=42.669, interpMode=RCIM_Cubic, leaveTangent=42.669, time=0.3135, value=8}
         {arriveTangent=-37.5281, interpMode=RCIM_Cubic, leaveTangent=-37.5281, time=0.3861, value=8}
         {arriveTangent=-11.5183, interpMode=RCIM_Cubic, leaveTangent=-11.5183, tangentMode=RCTM_User, time=0.5, value=1}
         <at 0.8 0 RCIM_Cubic>
        ]}}}
     CurveIndex=$Emitter.NormalizedLoopAge
     Scale Curve=2
    )
    Use Spawn Probability=true
   )
  ]
  particleSpawnScript [
   &InitializeParticle(
    Color=<rgba 0.1171 0.453 1 0.4>
    Lifetime=&RandomRangeFloat(Minimum=0.25, Maximum=0.55)
    Mesh Scale Mode=::NewEnumerator3
    Mesh Uniform Scale=1.1
    Sprite Size Mode=::NewEnumerator3
    Uniform Sprite Size=50
    UsePositionOffset=false
    Write Material Random=true
   )
   &InitialMeshOrientation(
    Mesh Orientation Mode=::NewEnumerator7
    Rotation=&MakeVector(X=&RandomRangeFloat(Minimum=0, Maximum=360), Y=0, Z=0)
    Use Rotation Vector=true
   )
   &AddVelocity(Velocity=<0 0 -20>) {enabled=false}
  ]
  particleUpdateScript [
   &ParticleState()
   &SkeletalMeshLocation(
    Bone Sampling Mode=::NewEnumerator4
    MaskSpawnGroup=false
    Sampled Position Offset=&MakeVector(
     X=0
     Y=0
     Z=&Lerp_Float(
      A=&RandomRangeFloat(Minimum=8, Maximum=16)
      B=&RandomRangeFloat(Minimum=-4, Maximum=-12)
      Alpha=$Particles.NormalizedAge
     )
    )
    Sampled Position Offset Coordinate Space:ENiagaraSkelMeshTransforms=::NewEnumerator2
    Skeletal Mesh=DataInterface() {propertyValues {
      DefaultMesh=null
      ExcludeBoneName=root
      OptionalComponentClass=null
      PreviewMesh=&M_Med_Soldier_04_Preview
      SkinningMode=SkinOnTheFly
      SoftSourceActor=null
      SourceMode=Default
      UvSetIndex=0
      WholeMeshLOD=-1
      bExcludeBone=false
      bReadDeformedGeometry=true
      bRequireCurrentFrameData=true
      ComponentTags []
      FilteredBones [hand_r]
      FilteredSockets []
      MeshUserParameter {parameter {name=None, typeDefHandle {registeredTypeIndex=1}, varData []}}
      SamplingRegions []
     }}
   ) {module=SkeletalMeshLocation001}
   /SetVariables(
    Particles.LightningLength:NiagaraFloat=&FloatFromCurve(
     FloatCurve=DataInterface() {propertyValues {Curve {keys [{arriveTangent=2.0099, interpMode=RCIM_Cubic, leaveTangent=2.0099, tangentMode=RCTM_User, time=0, value=0.5}, <at 1 1 RCIM_Cubic>]}}}
     Scale Curve=&RandomRangeFloat(Minimum=0.25, Maximum=0.75)
    )
    Particles.LightningShapeExpand:NiagaraFloat=&FloatFromCurve(
     FloatCurve=DataInterface() {propertyValues {Curve {keys [
         {arriveTangent=4.1785, interpMode=RCIM_Cubic, leaveTangent=4.1785, tangentMode=RCTM_User, time=0, value=0}
         {arriveTangent=1.1179, interpMode=RCIM_Cubic, leaveTangent=1.1179, tangentMode=RCTM_User, time=0.2042, value=0.6228}
         {arriveTangent=0.0224, interpMode=RCIM_Cubic, leaveTangent=0.0224, tangentMode=RCTM_User, time=1, value=1}
        ]}}}
     Scale Curve=&RandomRangeFloat(Minimum=0.25, Maximum=0.75)
    )
   ) {isSetParameters=true}
   &ScaleMeshSize(Scale Factor=&MakeVector(X=$Particles.LightningShapeExpand, Y=$Particles.LightningShapeExpand, Z=$Particles.LightningLength))
   /SetVariables(Particles.MeshOrientation=&MultiplyQuaternion(
     Quaternion A=&OrientMeshToVector(
      Look At Direction=&NormalizeVector(
       Vector To Normalize=&Subtract_Vector(
        A=Linked() {linkedVariable {name=Output.SkeletalMeshLocation002.SampledPositionWithoutOffset, type=NiagaraPosition}}
        B=Linked() {linkedVariable {name=Output.SkeletalMeshLocation.SampledPositionWithoutOffset, type=NiagaraPosition}}
       )
       Safe Normalize=true
       Safe Fallback Threshold=0
       Safe Fallback Vector=<0 0 0.001>
      )
      Reference Axis Vector=<1 0 0>
     )
     Quaternion B=&MakeQuatFromAxisAngle(Axis=<-1 0 0>, AngleInDegrees=&Multiply_Float(A=$Particles.NormalizedAge, B=150))
    )) {enabled=false, isSetParameters=true}
   &DynamicMaterialParameters(
    Index 0 Param 1=&FloatFromCurve(FloatCurve=DataInterface() {propertyValues {Curve {keys [<at 0 1>,<at 1 0.5>]}}}, Scale Curve=1.25)
    Index 0 Param 3=&RandomRangeFloat(Minimum=0, Maximum=22)
    Index 0 Param 4=0.45
    Index 1 Param 1=&FloatFromCurve(
     FloatCurve=DataInterface() {propertyValues {Curve {keys [
         <at 0 0>
         {arriveTangent=1, interpMode=RCIM_Cubic, leaveTangent=1, time=0.3, value=0}
         {interpMode=RCIM_Cubic, leaveTangent=-8.0089, time=1, value=1}
        ]}}}
     Scale Curve=0.1
    )
    Index 1 Param 2=1
    Index 1 Param 3=&FloatFromCurve(FloatCurve=DataInterface() {propertyValues {Curve {keys [<at 0.5 1 RCIM_Cubic>,<at 1 0 RCIM_Cubic>]}}}, Scale Curve=2)
    Index 2 Param 1=&FloatFromCurve(
     FloatCurve=DataInterface() {propertyValues {Curve {keys [{tangentMode=RCTM_User, time=0, value=0}, {tangentMode=RCTM_User, time=1, value=1}]}}}
     Scale Curve=0
    )
    Index 2 Param 2=2
    Module.Index0Param0Write=true
    Module.Index0Param1Write=true
    Module.Index0Param2Write=true
    Module.Index0Param3Write=true
    Module.Index1Param0Write=true
    Module.Index1Param1Write=true
    Module.Index1Param2Write=true
    Module.Index1Param3Write=true
    Module.Index2Param0Write=true
    Module.Index2Param1Write=true
    Module.Index2Param2Write=true
    Module.Index2Param3Write=true
    Write Parameter Index 1=true
    Write Parameter Index 2=true
   )
   &ScaleColor(
    Scale Alpha=&FloatFromCurve(FloatCurve=DataInterface() {propertyValues {Curve {keys [
         {tangentMode=RCTM_User, time=0, value=1}
         {arriveTangent=-0.0405, interpMode=RCIM_Cubic, leaveTangent=-0.0405, tangentMode=RCTM_User, time=0.7009, value=1}
         {arriveTangent=-8.0089, interpMode=RCIM_Cubic, leaveTangent=-8.0089, tangentMode=RCTM_User, time=1, value=0}
        ]}}})
    Scale RGB=&VectorFromFloat(Value=&FloatFromCurve(
      FloatCurve=DataInterface() {propertyValues {Curve {keys [
          {arriveTangent=9.0725, interpMode=RCIM_Cubic, time=0, value=0}
          <at 0.1408 2 RCIM_Cubic>
          {interpMode=RCIM_Cubic, leaveTangent=-0.0405, time=1, value=0}
         ]}}}
      Scale Curve=2.5
     ))
    ScaleA=true
    ScaleRGB=true
   ) {module=ScaleColor001}
   &UpdateMeshOrientation(Rotation Rate=25) {module=UpdateMeshOrientation002}
   &CameraOffset(Camera Offset Amount=5) {enabled=false}
   &SolveForcesAndVelocity(Clamp Velocity=false, Limit Acceleration=false)
  ]
  rendererClasses [&NiagaraMeshRendererProperties]
  renderers [{rendererClass=&NiagaraMeshRendererProperties, rendererIndex=0}]
 }
 GPU_particles001 {
  emitterUpdateScript [
   &EmitterState()
   &SpawnRate(
    SpawnRate=&FloatFromCurve(
     FloatCurve=DataInterface() {propertyValues {Curve {keys [
         <at 0.2299 1.3383 RCIM_Cubic>
         {arriveTangent=36.2637, interpMode=RCIM_Cubic, leaveTangent=36.2637, time=0.3135, value=7}
         {arriveTangent=-32.1668, interpMode=RCIM_Cubic, leaveTangent=-32.1668, time=0.3861, value=7}
         {arriveTangent=-16.9106, interpMode=RCIM_Cubic, leaveTangent=-16.9106, time=0.5, value=1}
         <at 0.8 0 RCIM_Cubic>
        ]}}}
     CurveIndex=$Emitter.NormalizedLoopAge
     Scale Curve=96
    )
    Use Spawn Probability=false
   )
  ]
  particleSpawnScript [
   &InitializeParticle(
    Color=&MakeCustomLinearColorFromBool(Bool=&RandomBool(), True Color=<rgba 10 40 100 1>, False Color=<rgba 5 68.3333 100 1>)
    Lifetime=/Niagara/DynamicInputs/Helpers/DistributionCurveFloat(
     Minimum:NiagaraFloat=0.2
     Maximum:NiagaraFloat=0.725
     Distribution Curve:NiagaraDataInterfaceCurve=DataInterface() {propertyValues {Curve {
        defaultValue=0
        keys [<at 0.5 0 RCIM_Cubic>, {arriveTangent=0.3606, interpMode=RCIM_Cubic, leaveTangent=0.3606, tangentMode=RCTM_User, time=1, value=1}]
       }}}
     Curve Parameter Type:NiagaraInt32=0
     "Curve Parameter Type:":NiagaraInt32=0
     Curve Parameter:NiagaraFloat=?
     Evaluation Type:ENiagaraRandomnessEvaluation=::NewEnumerator0
     Recalculate Random Each Loop:NiagaraBool=true
     Randomness Mode:ENiagaraRandomnessMode=::NewEnumerator0
     Random Seed:NiagaraInt32=0
     Fixed Random Seed:NiagaraBool=false
    )
    Sprite Rotation Mode=::NewEnumerator2
    Sprite Size Mode=::NewEnumerator1
    Uniform Sprite Size Max=2
    Uniform Sprite Size Min=10
    UsePositionOffset=false
    Write Material Random=true
   )
   &SkeletalMeshLocation(
    Bone Sampling Mode=::NewEnumerator4
    MaskSpawnGroup=false
    Skeletal Mesh=DataInterface() {propertyValues {
      DefaultMesh=null
      ExcludeBoneName=root
      OptionalComponentClass=null
      PreviewMesh=&M_Med_Soldier_04_Preview
      SkinningMode=SkinOnTheFly
      SoftSourceActor=null
      SourceMode=Default
      UvSetIndex=0
      WholeMeshLOD=-1
      bExcludeBone=false
      bReadDeformedGeometry=true
      bRequireCurrentFrameData=true
      ComponentTags []
      FilteredBones [hand_r]
      FilteredSockets []
      MeshUserParameter {parameter {name=None, typeDefHandle {registeredTypeIndex=1}, varData []}}
      SamplingRegions []
     }}
   )
   &ShapeLocation(
    Cylinder Height=24
    Cylinder Radius=4
    MaskSpawnGroup=false
    Override Seed=false
    Shape Origin=$Particles.SkeletalMeshLocation.SampledPosition
    Shape Primitive=::NewEnumerator1
    Surface Only=false
    UseOwnerScale=true
   )
   &AddVelocity(
    Cone Angle=90
    Cone Axis=<0 0 1>
    Module.UseConeFalloff=false
    Module.UseVelDistribution=false
    Override Seed=false
    Rotation Coordinate Space=::World
    Velocity Mode=::NewEnumerator2
    Velocity Speed=&RandomRangeFloat(Minimum=125, Maximum=175)
   )
  ]
  particleUpdateScript [
   &ParticleState()
   &ScaleSpriteSize(Uniform Curve Sprite Scale=DataInterface() {propertyValues {Curve {keys [
        {interpMode=RCIM_Cubic, tangentMode=RCTM_User, time=0, value=1}
        {arriveTangent=-1.9448, interpMode=RCIM_Cubic, leaveTangent=-1.9448, tangentMode=RCTM_User, time=1, value=0}
       ]}}})
   &CurlNoiseForce(
    Noise Strength=&FloatFromCurve(
     FloatCurve=DataInterface() {propertyValues {Curve {
        defaultValue=0
        keys [<at 0 0 RCIM_Cubic>, {arriveTangent=4.3146, interpMode=RCIM_Cubic, leaveTangent=4.3146, tangentMode=RCTM_User, time=0.3, value=1}]
       }}}
     Scale Curve=&RandomRangeFloat(Minimum=125, Maximum=325)
    )
    Pan Noise=false
   )
   &Drag(Drag=&RandomRangeFloat(Minimum=8, Maximum=12), Use Linear Drag=true, Use Rotational Drag=true)
   &SolveForcesAndVelocity(Clamp Velocity=false, Limit Acceleration=false)
   &ScaleColor(
    Scale Alpha=&FloatFromCurve(FloatCurve=DataInterface() {propertyValues {Curve {
        defaultValue=0
        keys [
         {interpMode=RCIM_Cubic, tangentMode=RCTM_User, time=0, value=0}
         {interpMode=RCIM_Cubic, tangentMode=RCTM_User, time=0.5, value=1}
         {tangentMode=RCTM_User, time=1, value=0}
        ]
       }}})
    Scale RGB=&Multiply_VectorByFloat(
     Vector=&VectorFromFloat(Value=&FNTimeOfDayScalar(Night Scale=0.75, Morning Scale=1, Evening Scale=1))
     Float=1
    )
    ScaleA=true
    ScaleRGB=true
   )
   /SetVariables(Particles.Glow Color=&MakeLinearColorFromVectorAndFloat(
     "Vector (RGB)"=&MakeVectorFromLinearColorRGB(Color=&MultiplyLinearColor(Linear Color=<rgba 0.0077 0.1731 1 1>, Scale Factor=&MakeVector4_FromFloat(Float=1)))
     "Float (Alpha)"=&FloatFromCurve(
      FloatCurve=DataInterface() {propertyValues {Curve {
         defaultValue=0
         keys [
          {interpMode=RCIM_Cubic, tangentMode=RCTM_User, time=0, value=0}
          {interpMode=RCIM_Cubic, tangentMode=RCTM_User, time=0.5, value=1}
          {tangentMode=RCTM_User, time=1, value=0}
         ]
        }}}
      Scale Curve=0.05
     )
    )) {isSetParameters=true}
   /SetVariables(Particles.Glow size:Vector2f=&Vector2DFromFloat(Value=&Multiply_Float(A=6, B=&FloatFromCurve(FloatCurve=DataInterface() {propertyValues {Curve {keys [<at 0 1>,<at 1 0>]}}})))) {isSetParameters=true}
  ]
  rendererClasses [&NiagaraSpriteRendererProperties]
  renderers [
   {rendererClass=&NiagaraSpriteRendererProperties}
   {rendererClass=&NiagaraSpriteRendererProperties, rendererIndex=1}
  ]
 }
 Hand_Glow {
  emitterUpdateScript [
   &EmitterState()
   &SpawnBurst_Instantaneous(Spawn Count=4, Use Spawn Probability=false, UseLoopCountLimit=false)
  ]
  particleSpawnScript [
   &FN_SetSpawnGroup()
   /SetVariables(Particles.VisibilityTag=&SpawnGroup()) {isSetParameters=true}
   &InitializeParticle(
    Color Mode=::NewEnumerator0
    Lifetime=&MergeHelper_FloatSwitch(0=5, 1=5, 2=5, 3=5, 4=0, Index=&SpawnGroup(), Num=::NewEnumerator3)
    UsePositionOffset=false
    Write Material Random=true
   )
   &CameraOffset(Camera Offset Amount=&MergeHelper_FloatSwitch(0=20, 1=10, 2=0, 3=0, 4=0, Index=&SpawnGroup(), Num=::NewEnumerator3))
  ]
  particleUpdateScript [
   &SolveForcesAndVelocity(Clamp Velocity=false, Limit Acceleration=false)
   &ParticleState()
   &SkeletalMeshLocation(
    Bone Sampling Mode=::NewEnumerator4
    MaskSpawnGroup=false
    Sampled Position Offset=&MergeHelper_VectorSwitch(0=<0 0 0>, 1=<0 0 6>, 2=<0 0 6>, 3=<0 0 0>, 4=<0 0 0>, Index=&SpawnGroup(), Num=::NewEnumerator3)
    Sampled Position Offset Coordinate Space:ENiagaraSkelMeshTransforms=::NewEnumerator2
    Skeletal Mesh=$System.Skeletal Mesh Reader
   )
   /SetVariables(
    Particles.Decal Color:LinearColor=&MultiplyLinearColor(
     Linear Color=<rgba 2.6271 19.2283 45 1>
     Scale Factor=&MakeVector4FromVectorAndFloat(
      "Vector (RGB)"=&VectorFromFloat(Value=&FloatFromCurve(FloatCurve=DataInterface() {propertyValues {Curve {keys [
            <at 0.36 1 RCIM_Cubic>
            <at 0.64 2 RCIM_Cubic>
            {arriveTangent=0.068, interpMode=RCIM_Cubic, leaveTangent=0.068, tangentMode=RCTM_User, time=0.7173, value=1}
            {arriveTangent=0.0869, interpMode=RCIM_Cubic, leaveTangent=0.0869, tangentMode=RCTM_User, time=0.8, value=1}
            <at 1 0 RCIM_Cubic>
           ]}}}))
      "Float (Alpha)"=&Multiply_Float(
       A=1
       B=&FloatFromCurve(FloatCurve=DataInterface() {propertyValues {Curve {keys [<at 0.8 1.0088 RCIM_Cubic>,<at 1 0 RCIM_Cubic>]}}})
      )
     )
    )
    Particles.Decal DP:Vector4f=&MakeVector4(
     X=&FloatFromCurve(
      FloatCurve=DataInterface() {propertyValues {Curve {keys [
          <at 0.36 0.8 RCIM_Cubic>
          {arriveTangent=-0.0248, interpMode=RCIM_Cubic, leaveTangent=-0.0248, tangentMode=RCTM_User, time=0.64, value=0.4}
          {arriveTangent=0.01, interpMode=RCIM_Cubic, leaveTangent=0.01, tangentMode=RCTM_User, time=0.7637, value=0.4}
          <at 1 1 RCIM_Cubic>
         ]}}}
      Scale Curve=4
     )
     Y=0.01
     Z=&FloatFromCurve(FloatCurve=DataInterface() {propertyValues {Curve {keys [<at 0 0.05>,<at 1 0>]}}})
     W=&RandomRangeFloat(Minimum=0, Maximum=1)
    )
    Particles.Decal Size:Vector2f=<30 30>
   ) {isSetParameters=true}
   /SetVariables(
    Particles.Inner Aura Color:LinearColor=&MultiplyLinearColor(
     Linear Color=<rgba 0.3103 2.9353 10 1>
     Scale Factor=&MakeVector4FromVectorAndFloat(
      "Vector (RGB)"=&VectorFromFloat(Value=&FloatFromCurve(
        FloatCurve=DataInterface() {propertyValues {Curve {keys [
            <at 0.36 1 RCIM_Cubic>
            <at 0.64 8 RCIM_Cubic>
            {arriveTangent=0.068, interpMode=RCIM_Cubic, leaveTangent=0.068, tangentMode=RCTM_User, time=0.7173, value=1}
            {arriveTangent=0.0869, interpMode=RCIM_Cubic, leaveTangent=0.0869, tangentMode=RCTM_User, time=0.8, value=1}
            <at 1 0 RCIM_Cubic>
           ]}}}
        Scale Curve=&FNTimeOfDayScalar(Night Scale=0.75, Morning Scale=1, Evening Scale=1)
       ))
      "Float (Alpha)"=&FloatFromCurve(
       FloatCurve=DataInterface() {propertyValues {Curve {keys [
           {arriveTangent=-0.0405, interpMode=RCIM_Cubic, leaveTangent=-0.0405, tangentMode=RCTM_User, time=0.8, value=1}
           {arriveTangent=-1.4489, interpMode=RCIM_Cubic, leaveTangent=-1.4489, tangentMode=RCTM_User, time=1, value=0}
          ]}}}
       Scale Curve=0.4
      )
     )
    )
    Particles.Inner Aura Size:Vector2f=&Multiply_Vector2D_ByFloat(Float=0.6, Vector2D=<35 60>)
   ) {isSetParameters=true}
   /SetVariables(
    Particles.Outer Aura Color:LinearColor=&MultiplyLinearColor(
     Linear Color=<rgba 0.3969 1.1104 2 1>
     Scale Factor=&MakeVector4FromVectorAndFloat(
      "Vector (RGB)"=&VectorFromFloat(Value=&FloatFromCurve(
        FloatCurve=DataInterface() {propertyValues {Curve {keys [
            <at 0.36 1 RCIM_Cubic>
            <at 0.64 4 RCIM_Cubic>
            {arriveTangent=0.068, interpMode=RCIM_Cubic, leaveTangent=0.068, tangentMode=RCTM_User, time=0.7173, value=1}
            {arriveTangent=0.0869, interpMode=RCIM_Cubic, leaveTangent=0.0869, tangentMode=RCTM_User, time=0.8, value=1}
            <at 1 0 RCIM_Cubic>
           ]}}}
        Scale Curve=&FNTimeOfDayScalar(Night Scale=0.75, Morning Scale=1, Evening Scale=1)
       ))
      "Float (Alpha)"=&FloatFromCurve(
       FloatCurve=DataInterface() {propertyValues {Curve {keys [
           {arriveTangent=-0.0405, interpMode=RCIM_Cubic, leaveTangent=-0.0405, tangentMode=RCTM_User, time=0.8, value=1}
           {arriveTangent=-1.4489, interpMode=RCIM_Cubic, leaveTangent=-1.4489, tangentMode=RCTM_User, time=1, value=0}
          ]}}}
       Scale Curve=0.4
      )
     )
    )
    Particles.Outer Aura Size:Vector2f=<40 60>
   ) {isSetParameters=true}
   /SetVariables(
    Particles.Glow Color=&MultiplyLinearColor(
     Linear Color=<rgba 0.0104 0.1135 0.5 0.25>
     Scale Factor=&MakeVector4FromVectorAndFloat(
      "Vector (RGB)"=&Multiply_VectorByFloat(Vector=<1 1 1>, Float=&FNTimeOfDayScalar(Night Scale=0.2, Morning Scale=0.5, Evening Scale=0.5))
      "Float (Alpha)"=&FloatFromCurve(FloatCurve=DataInterface() {propertyValues {Curve {defaultValue=1, keys [<at 0.8 1>,<at 0.9961 0>]}}}, Scale Curve=0.425)
     )
    )
    Particles.Glow DP:Vector4f=<1 5 0 0>
    Particles.Glow Size=&FN_FrontEndVector2Modifier(Front End:Vector2f=<0 0>, Gameplay:Vector2f=<35 80>)
   ) {isSetParameters=true}
   &AlignSpriteToMeshOrientation(
    Mesh Orientation Relative Sprite Alignment Vector=<-1 0 0>
    Mesh Orientation Relative Sprite Facing Vector:Vector3f=<0 1 0>
    SetAlignment=true
    SetFacing=true
   )
  ]
  rendererClasses [&NiagaraSpriteRendererProperties]
  renderers [
   {rendererClass=&NiagaraSpriteRendererProperties}
   {rendererClass=&NiagaraSpriteRendererProperties, rendererIndex=1}
   {rendererClass=&NiagaraSpriteRendererProperties, rendererIndex=2}
   {rendererClass=&NiagaraSpriteRendererProperties, rendererIndex=3}
  ]
 }
 Hand_Trail {
  emitterUpdateScript [
   &EmitterState()
   &SpawnRate(
    SpawnRate=&FloatFromCurve(
     FloatCurve=DataInterface() {propertyValues {Curve {keys [
         <at 2.9 0 RCIM_Cubic>
         {arriveTangent=2.8571, interpMode=RCIM_Cubic, leaveTangent=2.8571, time=3, value=1}
         {arriveTangent=-2.8571, interpMode=RCIM_Cubic, leaveTangent=-2.8571, time=3.25, value=1}
         <at 3.35 0 RCIM_Cubic>
        ]}}}
     CurveIndex=$Emitter.LoopedAge
     Scale Curve=60
    )
    Use Spawn Probability=false
   )
  ]
  particleSpawnScript [
   &InitializeParticle(
    Color=<rgba 1.3021 7.1092 25 1>
    Lifetime=0.125
    Ribbon Width=30
    Ribbon Width Mode=::NewEnumerator1
    UsePositionOffset=false
    Write Material Random=true
   )
   &SkeletalMeshLocation(
    Bone Sampling Mode=::NewEnumerator4
    MaskSpawnGroup=false
    Skeletal Mesh=DataInterface() {propertyValues {
      DefaultMesh=null
      ExcludeBoneName=root
      OptionalComponentClass=null
      PreviewMesh=&M_Med_Soldier_04_Preview
      SkinningMode=SkinOnTheFly
      SoftSourceActor=null
      SourceMode=Default
      UvSetIndex=0
      WholeMeshLOD=-1
      bExcludeBone=false
      bReadDeformedGeometry=true
      bRequireCurrentFrameData=true
      ComponentTags []
      FilteredBones [hand_r]
      FilteredSockets []
      MeshUserParameter {parameter {name=None, typeDefHandle {registeredTypeIndex=1}, varData []}}
      SamplingRegions []
     }}
   )
  ]
  particleUpdateScript [
   &ParticleState()
   &ScaleRibbonWidth(Ribbon Width Scale=&FloatFromCurve(FloatCurve=DataInterface() {propertyValues {Curve {keys [<at 0 1 RCIM_Cubic>, {arriveTangent=-2.0096, interpMode=RCIM_Cubic, leaveTangent=-2.0096, tangentMode=RCTM_User, time=1, value=0}]}}}))
   /SetVariables(
    Particles.LightningColor:LinearColor=&ColorFromCurve(ColorCurve=DataInterface() {propertyValues {
       AlphaCurve {keys [<at 0 0.65 RCIM_Cubic>,<at 1 0 RCIM_Cubic>]}
       BlueCurve {keys [<at 0 10 RCIM_Cubic>]}
       GreenCurve {keys [<at 0 5.1362 RCIM_Cubic>]}
       RedCurve {keys [<at 0 0.3125 RCIM_Cubic>]}
      }})
    Particles.LightningWidth:NiagaraFloat=&FloatFromCurve(FloatCurve=DataInterface() {propertyValues {Curve {keys [<at 0 1 RCIM_Cubic>,<at 1 0.45 RCIM_Cubic>]}}}, Scale Curve=20)
   ) {isSetParameters=true}
   &ScaleColor(
    Scale Alpha=&FloatFromCurve(
     FloatCurve=DataInterface() {propertyValues {Curve {
        defaultValue=1
        keys [{interpMode=RCIM_Cubic, tangentMode=RCTM_User, time=0, value=1}, {interpMode=RCIM_Cubic, tangentMode=RCTM_User, time=1, value=0}]
       }}}
     Scale Curve=0.4
    )
    ScaleA=true
    ScaleRGB=true
   )
  ]
  rendererClasses [&NiagaraRibbonRendererProperties]
  renderers [
   {rendererClass=&NiagaraRibbonRendererProperties}
   {rendererClass=&NiagaraRibbonRendererProperties, rendererIndex=1}
  ]
 }
 Shockwave {
  emitterUpdateScript [
   &EmitterState()
   &SpawnBurst_Instantaneous(Spawn Time=3.25, Use Spawn Probability=false, UseLoopCountLimit=false)
   &SpawnBurst_Instantaneous(Spawn Group=1, Spawn Time=3.25, Use Spawn Probability=false, UseLoopCountLimit=false) {module=SpawnBurst_Instantaneous001}
   &SpawnBurst_Instantaneous(Spawn Group=2, Spawn Time=3.25, Use Spawn Probability=false, UseLoopCountLimit=false) {module=SpawnBurst_Instantaneous002}
   &SpawnBurst_Instantaneous(Spawn Group=3, Spawn Time=3.25, Use Spawn Probability=false, UseLoopCountLimit=false) {module=SpawnBurst_Instantaneous003}
   &SpawnBurst_Instantaneous(Spawn Group=4, Spawn Time=3.25, Use Spawn Probability=false, UseLoopCountLimit=false) {module=SpawnBurst_Instantaneous004}
   &SpawnBurst_Instantaneous(Spawn Group=5, Spawn Time=3.25, Use Spawn Probability=false, UseLoopCountLimit=false) {module=SpawnBurst_Instantaneous005}
   &SpawnRate(
    SpawnGroup=5
    SpawnRate=&FloatFromCurve(
     FloatCurve=DataInterface() {propertyValues {Curve {keys [
         <at 3.25 0 RCIM_Cubic>
         {arriveTangent=0.1051, interpMode=RCIM_Cubic, leaveTangent=0.1051, tangentMode=RCTM_User, time=3.35, value=1}
         {arriveTangent=-0.0212, interpMode=RCIM_Cubic, leaveTangent=-0.0212, tangentMode=RCTM_User, time=5.6, value=1}
         <at 7 0 RCIM_Cubic>
        ]}}}
     CurveIndex=$Emitter.LoopedAge
     Scale Curve=1.2
    )
    Use Spawn Probability=false
   )
   &SpawnBurst_Instantaneous(Spawn Count=525, Spawn Group=6, Spawn Time=3.25, Use Spawn Probability=false, UseLoopCountLimit=false) {module=SpawnBurst_Instantaneous006}
   /SetVariables(Emitter.Refraction:NiagaraFloat=&FloatFromCurve(
     FloatCurve=DataInterface() {propertyValues {Curve {keys [
         {interpMode=RCIM_Cubic, leaveTangentWeight=0.0667, tangentMode=RCTM_User, time=0, value=0.825}
         {arriveTangent=0.2523, arriveTangentWeight=0.46, interpMode=RCIM_Cubic, leaveTangent=0.2523, tangentMode=RCTM_User, time=0.5, value=1}
        ]}}}
     CurveIndex=$Emitter.NormalizedLoopAge
    )) {isSetParameters=true}
  ]
  particleSpawnScript [
   /SetVariables(Particles.VisibilityTag=&SpawnGroup()) {isSetParameters=true}
   &InitializeParticle(
    Color Mode=::NewEnumerator0
    Lifetime=&MergeHelper_FloatSwitch(
     0=0.36
     1=0.36
     2=0.54
     3=0.24
     4=0.24
     5=1.6
     6=&DistributionCurveFloat(
      Minimum:NiagaraFloat=0.2
      Maximum:NiagaraFloat=0.725
      Distribution Curve:NiagaraDataInterfaceCurve=DataInterface() {propertyValues {Curve {
         defaultValue=0
         keys [<at 0.5 0 RCIM_Cubic>, {arriveTangent=0.3606, interpMode=RCIM_Cubic, leaveTangent=0.3606, tangentMode=RCTM_User, time=1, value=1}]
        }}}
      Curve Parameter Type:NiagaraInt32=0
      "Curve Parameter Type:":NiagaraInt32=0
      Curve Parameter:NiagaraFloat=?
      Evaluation Type:ENiagaraRandomnessEvaluation=::NewEnumerator0
      Recalculate Random Each Loop:NiagaraBool=true
      Randomness Mode:ENiagaraRandomnessMode=::NewEnumerator0
      Random Seed:NiagaraInt32=0
      Fixed Random Seed:NiagaraBool=false
     )
     Index=&SpawnGroup()
     Num=::NewEnumerator5
    )
    UsePositionOffset=false
    Write Material Random=true
   )
   &InitialMeshOrientation(
    Mesh Orientation Mode=::NewEnumerator2
    Rotation=&MakeVector(
     X=&MergeHelper_FloatSwitch(
      0=0
      1=0
      2=0
      3=&RandomRangeFloat(Minimum=0, Maximum=1)
      4=0
      Index=&SpawnGroup()
      Num=::NewEnumerator3
     )
     Y=&MergeHelper_FloatSwitch(0=0.25, 1=0, 2=0.25, 3=0, 4=0, Index=&SpawnGroup(), Num=::NewEnumerator3)
     Z=0
    )
    Use Rotation Vector=true
   )
   &CameraOffset(Camera Offset Amount=&MergeHelper_FloatSwitch(0=0, 1=40, 2=50, 3=0, 4=0, 5=-35, 6=0, Index=&SpawnGroup(), Num=::NewEnumerator5))
   &AddVelocity(
    Module.Use Falloff Distance=false
    Override Seed=false
    Velocity Mode=::NewEnumerator1
    Velocity Speed=&MergeHelper_FloatSwitch(
     0=0
     1=0
     2=0
     3=0
     4=0
     5=0
     6=&RandomRangeFloat(Minimum=125, Maximum=550)
     Index=&SpawnGroup()
     Num=::NewEnumerator5
    )
   )
  ]
  particleUpdateScript [
   &ParticleState()
   /SetVariables(
    Particles.Outer Shockwave Color:LinearColor=&MakeLinearColorFromVectorAndFloat(
     "Vector (RGB)"=&Multiply_VectorByFloat(
      Vector=&MakeVectorFromLinearColorRGB(Color=<rgba 1.0206 1.1647 2 1>)
      Float=&Multiply_Float(
       A=&FNTimeOfDayScalar(Night Scale=0.7, Morning Scale=0.3, Evening Scale=0.4)
       B=&FN_FrontEndFloatModifier(Gameplay=1, Front End=1)
      )
     )
     "Float (Alpha)"=&FloatFromCurve(FloatCurve=DataInterface() {propertyValues {Curve {keys [
          {interpMode=RCIM_Cubic, tangentMode=RCTM_User, time=0, value=0}
          {arriveTangent=-2.7304, interpMode=RCIM_Cubic, leaveTangent=-2.7304, tangentMode=RCTM_User, time=0.2, value=1}
          {tangentMode=RCTM_User, time=1, value=0}
         ]}}})
    )
    Particles.Outer Shockwave DP:Vector4f=&MakeVector4(
     X=-1
     Y=2
     Z=&FN_FrontEndFloatModifier(Gameplay=1, Front End=2)
     W=&FN_FrontEndFloatModifier(Gameplay=25, Front End=400)
    )
    Particles.Outer Shockwave Scale:Vector3f=&Multiply_VectorByFloat(
     Vector=<1 1 1>
     Float=&FloatFromCurve(
      FloatCurve=DataInterface() {propertyValues {Curve {keys [{interpMode=RCIM_Cubic, tangentMode=RCTM_User, time=0, value=0.4}, {interpMode=RCIM_Cubic, tangentMode=RCTM_User, time=1, value=1}]}}}
      Scale Curve=0.215
     )
    )
   ) {isSetParameters=true}
   /SetVariables(
    Particles.Inner Shockwave Color:LinearColor=&MakeLinearColorFromVectorAndFloat(
     "Vector (RGB)"=&MakeVectorFromLinearColorRGB(Color=<rgba 0.7319 0.9583 1 1>)
     "Float (Alpha)"=&FloatFromCurve(
      FloatCurve=DataInterface() {propertyValues {Curve {keys [{arriveTangent=-2.4409, arriveTangentWeight=0.162, interpMode=RCIM_Cubic, leaveTangentWeight=0.2268, time=0.2, value=1}, <at 1 0>]}}}
      Scale Curve=2.5
     )
    )
    Particles.Inner Shockwave DP:Vector4f=&MakeVector4(
     X=&FloatFromCurve(FloatCurve=DataInterface() {propertyValues {Curve {keys [
          {interpMode=RCIM_Cubic, tangentMode=RCTM_User, time=0.5, value=0}
          {arriveTangent=3.8908, interpMode=RCIM_Cubic, leaveTangent=3.8908, tangentMode=RCTM_User, time=1, value=1}
         ]}}})
     Y=&FloatFromCurve(FloatCurve=DataInterface() {propertyValues {Curve {keys [
          {arriveTangent=-0.6134, interpMode=RCIM_Cubic, leaveTangent=-0.6134, tangentMode=RCTM_User, time=0, value=3}
          {arriveTangent=-8.8976, interpMode=RCIM_Cubic, leaveTangent=-8.8976, tangentMode=RCTM_User, time=1, value=0.5}
         ]}}})
     Z=&FloatFromCurve(FloatCurve=DataInterface() {propertyValues {Curve {keys [
          {arriveTangent=-1.2425, interpMode=RCIM_Cubic, time=0, value=1}
          {arriveTangent=-2.1091, interpMode=RCIM_Cubic, leaveTangent=-2.1091, tangentMode=RCTM_User, time=1, value=0}
         ]}}})
     W=&FloatFromCurve(
      FloatCurve=DataInterface() {propertyValues {Curve {keys [
          {arriveTangent=2.3563, interpMode=RCIM_Cubic, leaveTangent=2.3563, tangentMode=RCTM_User, time=0, value=0}
          {tangentMode=RCTM_User, time=1, value=1}
         ]}}}
      Scale Curve=2
     )
    )
    Particles.Inner Shockwave Scale:Vector3f=&VectorFromFloat(Value=&FloatFromCurve(
      FloatCurve=DataInterface() {propertyValues {Curve {keys [
          {arriveTangent=0.7693, interpMode=RCIM_Cubic, leaveTangent=0.7693, tangentMode=RCTM_User, time=0, value=0.2}
          {arriveTangent=0.1864, interpMode=RCIM_Cubic, leaveTangent=0.1864, tangentMode=RCTM_User, time=1, value=1}
         ]}}}
      Scale Curve=1.25
     ))
   ) {isSetParameters=true}
   /SetVariables(
    Particles.Airflow Color:LinearColor=&MakeLinearColorFromVectorAndFloat(
     "Vector (RGB)"=&MakeVectorFromLinearColorRGB(Color=&ColorFromCurve(ColorCurve=DataInterface() {propertyValues {
         AlphaCurve {keys []}
         BlueCurve {keys [<at 0 5 RCIM_Cubic>,<at 0.6766 1 RCIM_Cubic>,<at 1 1>]}
         GreenCurve {keys [<at 0 1.954 RCIM_Cubic>,<at 0.6766 1 RCIM_Cubic>,<at 1 1>]}
         RedCurve {keys [<at 0 0.6009 RCIM_Cubic>,<at 0.6766 1 RCIM_Cubic>,<at 1 1>]}
        }}))
     "Float (Alpha)"=&FloatFromCurve(FloatCurve=DataInterface() {propertyValues {Curve {keys [
          {interpMode=RCIM_Cubic, tangentMode=RCTM_User, time=0, value=0}
          {arriveTangent=-3.4513, interpMode=RCIM_Cubic, leaveTangent=-3.4513, tangentMode=RCTM_User, time=0.15, value=1}
          {tangentMode=RCTM_User, time=1, value=0}
         ]}}})
    )
    Particles.Airflow DP:Vector4f=&MakeVector4(
     X=&FloatFromCurve(FloatCurve=DataInterface() {propertyValues {Curve {keys [{interpMode=RCIM_Cubic, tangentMode=RCTM_User, time=0.5, value=0}, {interpMode=RCIM_Cubic, tangentMode=RCTM_User, time=1, value=1}]}}})
     Y=&MakeCustomFloatFromBool(
      Bool=$User.IsFrontEnd
      True Float=0
      False Float=&FNTimeOfDayScalar(Night Scale=0.05, Morning Scale=-0.26, Day Scale=0, Evening Scale=-0.35)
     )
     Z=&MakeCustomFloatFromBool(
      Bool=$User.IsFrontEnd
      True Float=2
      False Float=&FNTimeOfDayScalar(Night Scale=0.5, Morning Scale=0.23, Evening Scale=0.2)
     )
     W=&MakeCustomFloatFromBool(
      Bool=$User.IsFrontEnd
      True Float=0
      False Float=&FNTimeOfDayScalar(Night Scale=0.02, Morning Scale=-0.01, Day Scale=0, Evening Scale=-0.01)
     )
    )
    Particles.Airflow Scale:Vector3f=&Multiply_VectorByFloat(
     Vector=&Multiply_Vector(
      A=&Multiply_VectorByFloat(
       Vector=&Multiply_VectorByFloat(Vector=<1 1 1>, Float=&Multiply_Float(A=0.67, B=3))
       Float=&FloatFromCurve(
        FloatCurve=DataInterface() {propertyValues {Curve {keys [
            {interpMode=RCIM_Cubic, tangentMode=RCTM_User, time=0, value=0.6}
            <at 0.4 0.725>
            {arriveTangent=-0.1262, interpMode=RCIM_Cubic, leaveTangent=-0.1262, tangentMode=RCTM_User, time=1, value=1}
           ]}}}
        Scale Curve=2
       )
      )
      B=&Multiply_VectorByFloat(
       Vector=<1 1 1>
       Float=&FloatFromCurve(FloatCurve=DataInterface() {propertyValues {Curve {keys [
            {arriveTangent=1.3474, interpMode=RCIM_Cubic, leaveTangent=1.3474, tangentMode=RCTM_User, time=0, value=0.5}
            {arriveTangent=0.7, interpMode=RCIM_Cubic, leaveTangent=0.7, time=0.2553, value=0.9138}
            <at 1 1.2>
           ]}}})
      )
     )
     Float=0.325
    )
   ) {isSetParameters=true}
   /SetVariables(
    Particles.Flash Color:LinearColor=&MakeLinearColorFromVectorAndFloat(
     "Vector (RGB)"=&Multiply_Vector(
      A=&MakeVectorFromLinearColorRGB(Color=&ColorFromCurve(ColorCurve=DataInterface() {propertyValues {
          AlphaCurve {keys []}
          BlueCurve {keys [<at 0 1 RCIM_Cubic>,<at 1 0 RCIM_Cubic>]}
          GreenCurve {keys [<at 0 1 RCIM_Cubic>,<at 1 0 RCIM_Cubic>]}
          RedCurve {keys [<at 0 1 RCIM_Cubic>,<at 1 0 RCIM_Cubic>]}
         }}))
      B=&MakeVectorFromLinearColorRGB(Color=<rgba 0.7501 1.9358 5 1>)
     )
     "Float (Alpha)"=&FloatFromCurve(FloatCurve=DataInterface() {propertyValues {Curve {
         defaultValue=1
         keys [
          {arriveTangent=9.0725, interpMode=RCIM_Cubic, leaveTangent=9.0725, tangentMode=RCTM_User, time=0, value=0}
          {arriveTangent=-2.5249, interpMode=RCIM_Cubic, leaveTangent=-2.5249, tangentMode=RCTM_User, time=0.1408, value=0.9847}
          {arriveTangent=-0.0405, interpMode=RCIM_Cubic, leaveTangent=-0.0405, tangentMode=RCTM_User, time=1, value=0}
         ]
        }}})
    )
    Particles.Flash Scale:Vector3f=&VectorFromFloat(Value=&FloatFromCurve(
      FloatCurve=DataInterface() {propertyValues {Curve {
         defaultValue=1
         keys [
          {arriveTangent=9.0725, interpMode=RCIM_Cubic, leaveTangent=9.0725, tangentMode=RCTM_User, time=0, value=0}
          {arriveTangent=-2.5249, interpMode=RCIM_Cubic, leaveTangent=-2.5249, tangentMode=RCTM_User, time=0.1408, value=0.9847}
          {arriveTangent=-0.0405, interpMode=RCIM_Cubic, leaveTangent=-0.0405, tangentMode=RCTM_User, time=1, value=0}
         ]
        }}}
      Scale Curve=0.7
     ))
   ) {isSetParameters=true}
   /SetVariables(
    Particles.Glow Color=&Multiply_LinearColorByFloat(
     Linear Color=&MultiplyLinearColor(
      Linear Color=<rgba 0.709 2.174 5 0.3>
      Scale Factor=&MakeVector4_FromLinearColor(LinearColor=&ColorFromCurve(ColorCurve=DataInterface() {propertyValues {
          AlphaCurve {keys [<at 0 0>,<at 0.2 1 RCIM_Cubic>,<at 1 0 RCIM_Cubic>]}
          BlueCurve {keys [<at 0 1 RCIM_Cubic>,<at 1 0 RCIM_Cubic>]}
          GreenCurve {keys [<at 0 1 RCIM_Cubic>,<at 1 0 RCIM_Cubic>]}
          RedCurve {keys [<at 0 1 RCIM_Cubic>,<at 1 0 RCIM_Cubic>]}
         }}))
     )
     Float=&FN_FrontEndFloatModifier(Gameplay=1, Front End=0.45)
    )
    Particles.Glow Size=&Vector2DFromFloat(Value=&FloatFromCurve(
      FloatCurve=DataInterface() {propertyValues {Curve {
         defaultValue=1
         keys [
          {arriveTangent=9.0725, interpMode=RCIM_Cubic, leaveTangent=9.0725, tangentMode=RCTM_User, time=0, value=0}
          {arriveTangent=-2.5249, interpMode=RCIM_Cubic, leaveTangent=-2.5249, tangentMode=RCTM_User, time=0.1408, value=0.9847}
          {arriveTangent=-0.0405, interpMode=RCIM_Cubic, leaveTangent=-0.0405, tangentMode=RCTM_User, time=1, value=0}
         ]
        }}}
      Scale Curve=300
     ))
   ) {isSetParameters=true}
   /SetVariables(
    Particles.Refraction Color:LinearColor=&MakeLinearColorFromVectorAndFloat(
     "Vector (RGB)"=&MakeVectorFromLinearColorRGB(Color=<rgba 0.0904 0.3335 1 0.4>)
     "Float (Alpha)"=&FloatFromCurve(
      FloatCurve=DataInterface() {propertyValues {Curve {keys [
          {arriveTangent=-0.0761, interpMode=RCIM_Cubic, leaveTangent=-0.0761, leaveTangentWeight=0.9539, tangentMode=RCTM_User, time=0, value=1}
          {arriveTangentWeight=1.1112, interpMode=RCIM_Cubic, tangentMode=RCTM_User, time=1, value=0}
         ]}}}
      Scale Curve=&FloatFromCurve(
       FloatCurve=DataInterface() {propertyValues {Curve {keys [{arriveTangent=-1.6376, interpMode=RCIM_Cubic, leaveTangent=-1.6376, tangentMode=RCTM_User, time=0.2, value=1}, <at 1 0 RCIM_Cubic>]}}}
       CurveIndex=$Emitter.NormalizedLoopAge
      )
     )
    )
    Particles.Refraction DP:Vector4f=<0 0 0 0>
    Particles.Refraction Size:Vector2f=&Vector2DFromFloat(Value=&FloatFromCurve(
      FloatCurve=DataInterface() {propertyValues {Curve {keys [
          {arriveTangent=4.1785, interpMode=RCIM_Cubic, leaveTangent=4.1785, tangentMode=RCTM_User, time=0, value=0}
          {arriveTangent=0.9995, interpMode=RCIM_Cubic, leaveTangent=0.9995, tangentMode=RCTM_User, time=0.2223, value=0.6825}
          {arriveTangent=-0.0405, interpMode=RCIM_Cubic, leaveTangent=-0.0405, tangentMode=RCTM_User, time=1, value=1}
         ]}}}
      Scale Curve=180
     ))
   ) {isSetParameters=true}
   /SetVariables(
    Particles.Particle 1 Color:LinearColor=&MakeLinearColorFromVectorAndFloat(
     "Vector (RGB)"=&Multiply_VectorByFloat(
      Vector=&MakeVectorFromLinearColorRGB(Color=&MakeCustomLinearColorFromBool(Bool=&RandomBool(), True Color=<rgba 10 40 100 1>, False Color=<rgba 5 68.3333 100 1>))
      Float=&FNTimeOfDayScalar(Night Scale=0.75, Morning Scale=1, Evening Scale=1)
     )
     "Float (Alpha)"=&FloatFromCurve(FloatCurve=DataInterface() {propertyValues {Curve {
         defaultValue=0
         keys [
          {interpMode=RCIM_Cubic, tangentMode=RCTM_User, time=0, value=0}
          {interpMode=RCIM_Cubic, tangentMode=RCTM_User, time=0.5, value=1}
          {tangentMode=RCTM_User, time=1, value=0}
         ]
        }}})
    )
    Particles.Particle 1 Size:Vector2f=&Vector2DFromFloat(Value=&FloatFromCurve(
      FloatCurve=DataInterface() {propertyValues {Curve {keys [
          {interpMode=RCIM_Cubic, tangentMode=RCTM_User, time=0, value=1}
          {arriveTangent=-1.9448, interpMode=RCIM_Cubic, leaveTangent=-1.9448, tangentMode=RCTM_User, time=1, value=0}
         ]}}}
      Scale Curve=&RandomRangeFloat(Minimum=2, Maximum=10)
     ))
   ) {isSetParameters=true}
   /SetVariables(
    Particles.Particle 2 Color:LinearColor=&MakeLinearColorFromVectorAndFloat(
     "Vector (RGB)"=&MakeVectorFromLinearColorRGB(Color=&MultiplyLinearColor(Linear Color=<rgba 0.0077 0.1731 1 1>, Scale Factor=&MakeVector4_FromFloat(Float=1)))
     "Float (Alpha)"=&FloatFromCurve(
      FloatCurve=DataInterface() {propertyValues {Curve {
         defaultValue=0
         keys [
          {interpMode=RCIM_Cubic, tangentMode=RCTM_User, time=0, value=0}
          {interpMode=RCIM_Cubic, tangentMode=RCTM_User, time=0.5, value=1}
          {tangentMode=RCTM_User, time=1, value=0}
         ]
        }}}
      Scale Curve=0.05
     )
    )
    Particles.Particle 2 Size:Vector2f=&Vector2DFromFloat(Value=&Multiply_Float(A=6, B=&FloatFromCurve(FloatCurve=DataInterface() {propertyValues {Curve {keys [<at 0 1>,<at 1 0>]}}})))
   ) {isSetParameters=true}
   &UpdateMeshOrientation(
    Rotation Rate=&MergeHelper_FloatSwitch(0=0, 1=0, 2=1, 3=0, 4=0, Index=&SpawnGroup(), Num=::NewEnumerator3)
    Rotation Vector=<0 0 1>
   )
   &CurlNoiseForce(
    Noise Strength=&MergeHelper_FloatSwitch(
     0=0
     1=0
     2=0
     3=0
     4=0
     5=0
     6=&FloatFromCurve(
      FloatCurve=DataInterface() {propertyValues {Curve {
         defaultValue=0
         keys [<at 0 0 RCIM_Cubic>, {arriveTangent=4.3146, interpMode=RCIM_Cubic, leaveTangent=4.3146, tangentMode=RCTM_User, time=0.3, value=1}]
        }}}
      Scale Curve=&RandomRangeFloat(Minimum=325, Maximum=700)
     )
     Index=&SpawnGroup()
     Num=::NewEnumerator5
    )
    Pan Noise=false
   )
   &Drag(
    Drag=&MergeHelper_FloatSwitch(
     0=0
     1=0
     2=0
     3=0
     4=0
     5=0
     6=&RandomRangeFloat(Minimum=2, Maximum=5)
     Index=&SpawnGroup()
     Num=::NewEnumerator5
    )
    Use Linear Drag=true
    Use Rotational Drag=true
   )
   &SolveForcesAndVelocity(Clamp Velocity=false, Limit Acceleration=false)
  ]
  rendererClasses [&NiagaraMeshRendererProperties, &NiagaraSpriteRendererProperties]
  renderers [
   {rendererClass=&NiagaraMeshRendererProperties, rendererIndex=0}
   {rendererClass=&NiagaraMeshRendererProperties, rendererIndex=1}
   {rendererClass=&NiagaraMeshRendererProperties, rendererIndex=2}
   {rendererClass=&NiagaraMeshRendererProperties, rendererIndex=3}
   {rendererClass=&NiagaraSpriteRendererProperties, rendererIndex=4}
   {rendererClass=&NiagaraSpriteRendererProperties, rendererIndex=5}
   {rendererClass=&NiagaraSpriteRendererProperties, rendererIndex=6}
   {rendererClass=&NiagaraSpriteRendererProperties, rendererIndex=7}
  ]
 }
}
summary {
 systemName=NS_Emote_FoundKelp_Boom
 emitters [
  {emitterName=ElectricCoils, rendererClasses [&NiagaraMeshRendererProperties]}
  {emitterName=GPU_particles001, rendererClasses [&NiagaraSpriteRendererProperties]}
  {emitterName=Hand_Trail, rendererClasses [&NiagaraRibbonRendererProperties]}
  {emitterName=Shockwave, rendererClasses [&NiagaraMeshRendererProperties, &NiagaraSpriteRendererProperties]}
  {emitterName=Hand_Glow, rendererClasses [&NiagaraSpriteRendererProperties]}
 ]
 userVariables [
  User.bIsFrontEnd {type=NiagaraBool}
  User.bIsFrontEndPreview {type=NiagaraBool}
  User.BlinkIndex {type=NiagaraFloat, defaultValue=0, description=""}
  User.IsFrontEnd {type=NiagaraBool}
 ]
}
