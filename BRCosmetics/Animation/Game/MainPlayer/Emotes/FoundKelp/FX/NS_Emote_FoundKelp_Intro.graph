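# Niagara effect for the FoundKelp emote intro that builds a blue electrical charge on the character's hand.
# ElectricCoils spawns short-lived, expanding lightning meshes that slide along hand_r with animated material parameters.
# GPU_particles001 ramps up sparks from a cylinder at hand_r, lifted by curl noise and upward gravity.
# Hand_Glow bursts four skeleton-attached sprites for decal, auras and glow, fading in by emitter age with time-of-day dimming and front-end glow sizing.
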NiagaraSystem /BRCosmetics/Animation/Game/MainPlayer/Emotes/FoundKelp/FX/NS_Emote_FoundKelp_Intro
dependencies {
 usedDataInterfaces [&NiagaraDataInterfaceSkeletalMesh, &NiagaraDataInterfaceCurve]
 usedDynamicInputs [&RandomRangeVector]
 usedModules [
  &self:SystemSpawnScript.SystemScriptSource.SystemScriptGraph.NiagaraNodeAssignment_0.SetVariables_#1
  &SystemState
 ]
}
emitters {
 ElectricCoils {
  emitterUpdateScript [
   &EmitterState(Life Cycle Mode=::NewEnumerator1, Loop Behavior=::NewEnumerator1, Loop Duration Mode=::NewEnumerator1, UseLoopDelay=false)
   &SpawnRate(Spawn Probability=0.525, SpawnRate=2, Use Spawn Probability=true)
  ]
  particleSpawnScript [
   &FN_SetSpawnGroup(Number of Spawn Groups=2)
   &InitializeParticle(
    Color=<rgba 0.1171 0.453 1 0.4>
    Lifetime=&RandomRangeFloat(Minimum=0.25, Maximum=0.55)
    Mesh Scale Mode=::NewEnumerator3
    Mesh Uniform Scale=1.1
    Sprite Size Mode=::NewEnumerator3
    Uniform Sprite Size=50
    UsePositionOffset=false
    Write Material Random=true
   )
   &InitialMeshOrientation(
    Mesh Orientation Mode=::NewEnumerator7
    Rotation=&MakeVector(X=&RandomRangeFloat(Minimum=0, Maximum=360), Y=0, Z=0)
    Use Rotation Vector=true
   )
   &AddVelocity(Velocity=<0 0 -20>) {enabled=false}
  ]
  particleUpdateScript [
   &ParticleState()
   &SkeletalMeshLocation(
    Bone Sampling Mode=::NewEnumerator4
    MaskSpawnGroup=false
    Sampled Position Offset=&MakeVector(
     X=0
     Y=0
     Z=&Lerp_Float(
      A=&RandomRangeFloat(Minimum=8, Maximum=16)
      B=&RandomRangeFloat(Minimum=-4, Maximum=-12)
      Alpha=$Particles.NormalizedAge
     )
    )
    Sampled Position Offset Coordinate Space:ENiagaraSkelMeshTransforms=::NewEnumerator2
    Skeletal Mesh=DataInterface() {propertyValues {
      DefaultMesh=null
      ExcludeBoneName=root
      OptionalComponentClass=null
      PreviewMesh=&M_Med_Soldier_04_Preview
      SkinningMode=SkinOnTheFly
      SoftSourceActor=null
      SourceMode=Default
      UvSetIndex=0
      WholeMeshLOD=-1
      bExcludeBone=false
      bReadDeformedGeometry=true
      bRequireCurrentFrameData=true
      ComponentTags []
      FilteredBones [hand_r]
      FilteredSockets []
      MeshUserParameter {parameter {name=None, typeDefHandle {registeredTypeIndex=1}, varData []}}
      SamplingRegions []
     }}
   ) {module=SkeletalMeshLocation001}
   /SetVariables(
    Particles.LightningLength:NiagaraFloat=&FloatFromCurve(
     FloatCurve=DataInterface() {propertyValues {Curve {keys [{arriveTangent=2.0099, interpMode=RCIM_Cubic, leaveTangent=2.0099, tangentMode=RCTM_User, time=0, value=0.5}, <at 1 1 RCIM_Cubic>]}}}
     Scale Curve=&RandomRangeFloat(Minimum=0.25, Maximum=0.75)
    )
    Particles.LightningShapeExpand:NiagaraFloat=&FloatFromCurve(
     FloatCurve=DataInterface() {propertyValues {Curve {keys [
         {arriveTangent=4.1785, interpMode=RCIM_Cubic, leaveTangent=4.1785, tangentMode=RCTM_User, time=0, value=0}
         {arriveTangent=1.1179, interpMode=RCIM_Cubic, leaveTangent=1.1179, tangentMode=RCTM_User, time=0.2042, value=0.6228}
         {arriveTangent=0.0224, interpMode=RCIM_Cubic, leaveTangent=0.0224, tangentMode=RCTM_User, time=1, value=1}
        ]}}}
     Scale Curve=&RandomRangeFloat(Minimum=0.25, Maximum=0.75)
    )
   ) {isSetParameters=true}
   &ScaleMeshSize(Scale Factor=&MakeVector(X=$Particles.LightningShapeExpand, Y=$Particles.LightningShapeExpand, Z=$Particles.LightningLength))
   /SetVariables(Particles.MeshOrientation=&MultiplyQuaternion(
     Quaternion A=&OrientMeshToVector(
      Look At Direction=&NormalizeVector(
       Vector To Normalize=&Subtract_Vector(
        A=Linked() {linkedVariable {name=Output.SkeletalMeshLocation002.SampledPositionWithoutOffset, type=NiagaraPosition}}
        B=Linked() {linkedVariable {name=Output.SkeletalMeshLocation.SampledPositionWithoutOffset, type=NiagaraPosition}}
       )
       Safe Normalize=true
       Safe Fallback Threshold=0
       Safe Fallback Vector=<0 0 0.001>
      )
      Reference Axis Vector=<1 0 0>
     )
     Quaternion B=&MakeQuatFromAxisAngle(Axis=<-1 0 0>, AngleInDegrees=&Multiply_Float(A=$Particles.NormalizedAge, B=150))
    )) {enabled=false, isSetParameters=true}
   &DynamicMaterialParameters(
    Index 0 Param 1=&FloatFromCurve(FloatCurve=DataInterface() {propertyValues {Curve {keys [<at 0 1>,<at 1 0.5>]}}}, Scale Curve=1.25)
    Index 0 Param 3=&RandomRangeFloat(Minimum=0, Maximum=22)
    Index 0 Param 4=0.45
    Index 1 Param 1=&FloatFromCurve(
     FloatCurve=DataInterface() {propertyValues {Curve {keys [
         <at 0 0>
         {arriveTangent=1, interpMode=RCIM_Cubic, leaveTangent=1, time=0.3, value=0}
         {interpMode=RCIM_Cubic, leaveTangent=-8.0089, time=1, value=1}
        ]}}}
     Scale Curve=0.1
    )
    Index 1 Param 2=1
    Index 1 Param 3=&FloatFromCurve(FloatCurve=DataInterface() {propertyValues {Curve {keys [<at 0.5 1 RCIM_Cubic>,<at 1 0 RCIM_Cubic>]}}}, Scale Curve=2)
    Index 2 Param 1=&FloatFromCurve(
     FloatCurve=DataInterface() {propertyValues {Curve {keys [{tangentMode=RCTM_User, time=0, value=0}, {tangentMode=RCTM_User, time=1, value=1}]}}}
     Scale Curve=0
    )
    Index 2 Param 2=2
    Module.Index0Param0Write=true
    Module.Index0Param1Write=true
    Module.Index0Param2Write=true
    Module.Index0Param3Write=true
    Module.Index1Param0Write=true
    Module.Index1Param1Write=true
    Module.Index1Param2Write=true
    Module.Index1Param3Write=true
    Module.Index2Param0Write=true
    Module.Index2Param1Write=true
    Module.Index2Param2Write=true
    Module.Index2Param3Write=true
    Write Parameter Index 1=true
    Write Parameter Index 2=true
   )
   &ScaleColor(
    Scale Alpha=&FloatFromCurve(FloatCurve=DataInterface() {propertyValues {Curve {keys [
         {tangentMode=RCTM_User, time=0, value=1}
         {arriveTangent=-0.0405, interpMode=RCIM_Cubic, leaveTangent=-0.0405, tangentMode=RCTM_User, time=0.7009, value=1}
         {arriveTangent=-8.0089, interpMode=RCIM_Cubic, leaveTangent=-8.0089, tangentMode=RCTM_User, time=1, value=0}
        ]}}})
    Scale RGB=&VectorFromFloat(Value=&FloatFromCurve(
      FloatCurve=DataInterface() {propertyValues {Curve {keys [
          {arriveTangent=9.0725, interpMode=RCIM_Cubic, time=0, value=0}
          <at 0.1408 2 RCIM_Cubic>
          {interpMode=RCIM_Cubic, leaveTangent=-0.0405, time=1, value=0}
         ]}}}
      Scale Curve=2.5
     ))
    ScaleA=true
    ScaleRGB=true
   ) {module=ScaleColor001}
   &UpdateMeshOrientation(Rotation Rate=25) {module=UpdateMeshOrientation002}
   &CameraOffset(Camera Offset Amount=5) {enabled=false}
   &SolveForcesAndVelocity(Clamp Velocity=false, Limit Acceleration=false)
  ]
  rendererClasses [&NiagaraMeshRendererProperties]
  renderers [{rendererClass=&NiagaraMeshRendererProperties, rendererIndex=0}]
 }
 GPU_particles001 {
  emitterUpdateScript [
   &EmitterState(Life Cycle Mode=::NewEnumerator1, Loop Behavior=::NewEnumerator1, Loop Duration Mode=::NewEnumerator1, UseLoopDelay=false)
   &SpawnRate(
    Spawn Probability=0.8
    SpawnRate=&FloatFromCurve(
     FloatCurve=DataInterface() {propertyValues {Curve {keys [<at 0 0 RCIM_Cubic>,<at 0.45 1 RCIM_Cubic>]}}}
     CurveIndex=$Emitter.Age
     Scale Curve=96
    )
    Use Spawn Probability=true
   )
  ]
  particleSpawnScript [
   &InitializeParticle(
    Color=&MakeCustomLinearColorFromBool(Bool=&RandomBool(), True Color=<rgba 10 40 100 1>, False Color=<rgba 5 68.3333 100 1>)
    Lifetime=/Niagara/DynamicInputs/Helpers/DistributionCurveFloat(
     Minimum:NiagaraFloat=0.3
     Maximum:NiagaraFloat=1
     Distribution Curve:NiagaraDataInterfaceCurve=DataInterface() {propertyValues {Curve {
        defaultValue=0
        keys [<at 0.65 0 RCIM_Cubic>, {arriveTangent=0.3606, interpMode=RCIM_Cubic, leaveTangent=0.3606, tangentMode=RCTM_User, time=1, value=1}]
       }}}
     Curve Parameter Type:NiagaraInt32=0
     "Curve Parameter Type:":NiagaraInt32=0
     Curve Parameter:NiagaraFloat=?
     Evaluation Type:ENiagaraRandomnessEvaluation=::NewEnumerator0
     Recalculate Random Each Loop:NiagaraBool=true
     Randomness Mode:ENiagaraRandomnessMode=::NewEnumerator0
     Random Seed:NiagaraInt32=0
     Fixed Random Seed:NiagaraBool=false
    )
    Sprite Rotation Mode=::NewEnumerator2
    Sprite Size Mode=::NewEnumerator1
    Uniform Sprite Size Max=6
    Uniform Sprite Size Min=14
    UsePositionOffset=false
    Write Material Random=true
   )
   &SkeletalMeshLocation(
    Bone Sampling Mode=::NewEnumerator4
    MaskSpawnGroup=false
    Skeletal Mesh=DataInterface() {propertyValues {
      DefaultMesh=null
      ExcludeBoneName=root
      OptionalComponentClass=null
      PreviewMesh=&M_Med_Soldier_04_Preview
      SkinningMode=SkinOnTheFly
      SoftSourceActor=null
      SourceMode=Default
      UvSetIndex=0
      WholeMeshLOD=-1
      bExcludeBone=false
      bReadDeformedGeometry=true
      bRequireCurrentFrameData=true
      ComponentTags []
      FilteredBones [hand_r]
      FilteredSockets []
      MeshUserParameter {parameter {name=None, typeDefHandle {registeredTypeIndex=1}, varData []}}
      SamplingRegions []
     }}
   )
   &ShapeLocation(
    Cylinder Height=24
    Cylinder Radius=4
    MaskSpawnGroup=false
    Override Seed=false
    Shape Origin=$Particles.SkeletalMeshLocation.SampledPosition
    Shape Primitive=::NewEnumerator1
    Surface Only=false
    UseOwnerScale=true
   )
   &AddVelocity(
    Cone Angle=90
    Cone Axis=<0 0 1>
    Module.UseConeFalloff=false
    Module.UseVelDistribution=false
    Override Seed=false
    Rotation Coordinate Space=::World
    Velocity Mode=::NewEnumerator2
    Velocity Speed=&RandomRangeFloat(Minimum=125, Maximum=175)
   )
  ]
  particleUpdateScript [
   &ParticleState()
   &ScaleSpriteSize(Uniform Curve Sprite Scale=DataInterface() {propertyValues {Curve {keys [
        {interpMode=RCIM_Cubic, tangentMode=RCTM_User, time=0, value=1}
        {arriveTangent=-0.0158, interpMode=RCIM_Cubic, leaveTangent=-0.0158, tangentMode=RCTM_User, time=1, value=0}
       ]}}})
   &CurlNoiseForce(
    Noise Strength=&FloatFromCurve(
     FloatCurve=DataInterface() {propertyValues {Curve {
        defaultValue=0
        keys [<at 0 0 RCIM_Cubic>, {arriveTangent=4.3146, interpMode=RCIM_Cubic, leaveTangent=4.3146, tangentMode=RCTM_User, time=0.3, value=1}]
       }}}
     Scale Curve=&RandomRangeFloat(Minimum=125, Maximum=325)
    )
    Pan Noise=false
   )
   &Drag(Drag=&RandomRangeFloat(Minimum=8, Maximum=12), Use Linear Drag=true, Use Rotational Drag=true)
   &GravityForce(Gravity=&MakeVector(X=0, Y=0, Z=&RandomRangeFloat(Minimum=100, Maximum=225)))
   &SolveForcesAndVelocity(Clamp Velocity=false, Limit Acceleration=false)
   &ScaleColor(
    Scale Alpha=&FloatFromCurve(FloatCurve=DataInterface() {propertyValues {Curve {
        defaultValue=0
        keys [
         {interpMode=RCIM_Cubic, tangentMode=RCTM_User, time=0, value=0}
         {interpMode=RCIM_Cubic, tangentMode=RCTM_User, time=0.5, value=1}
         {tangentMode=RCTM_User, time=1, value=0}
        ]
       }}})
    Scale RGB=&Multiply_VectorByFloat(
     Vector=&VectorFromFloat(Value=&FNTimeOfDayScalar(Night Scale=0.75, Morning Scale=1, Evening Scale=1))
     Float=1
    )
    ScaleA=true
    ScaleRGB=true
   )
   /SetVariables(Particles.Glow Color=&MakeLinearColorFromVectorAndFloat(
     "Vector (RGB)"=&MakeVectorFromLinearColorRGB(Color=&MultiplyLinearColor(Linear Color=<rgba 0.0077 0.1731 1 1>, Scale Factor=&MakeVector4_FromFloat(Float=1)))
     "Float (Alpha)"=&FloatFromCurve(
      FloatCurve=DataInterface() {propertyValues {Curve {
         defaultValue=0
         keys [
          {interpMode=RCIM_Cubic, tangentMode=RCTM_User, time=0, value=0}
          {interpMode=RCIM_Cubic, tangentMode=RCTM_User, time=0.5, value=1}
          {tangentMode=RCTM_User, time=1, value=0}
         ]
        }}}
      Scale Curve=0.05
     )
    )) {isSetParameters=true}
   /SetVariables(Particles.Glow size:Vector2f=&Vector2DFromFloat(Value=&Multiply_Float(
      A=8
      B=&FloatFromCurve(FloatCurve=DataInterface() {propertyValues {Curve {keys [<at 0 1 RCIM_Cubic>,<at 1 0 RCIM_Cubic>]}}})
     ))) {isSetParameters=true}
  ]
  rendererClasses [&NiagaraSpriteRendererProperties]
  renderers [
   {rendererClass=&NiagaraSpriteRendererProperties}
   {rendererClass=&NiagaraSpriteRendererProperties, rendererIndex=1}
  ]
 }
 Hand_Glow {
  emitterUpdateScript [
   &EmitterState(Life Cycle Mode=::NewEnumerator1, Loop Behavior=::NewEnumerator1, Loop Duration Mode=::NewEnumerator1, UseLoopDelay=false)
   &SpawnBurst_Instantaneous(Spawn Count=4, Use Spawn Probability=false, UseLoopCountLimit=false)
  ]
  particleSpawnScript [
   &FN_SetSpawnGroup()
   /SetVariables(Particles.VisibilityTag=&SpawnGroup()) {isSetParameters=true}
   &InitializeParticle(Color Mode=::NewEnumerator0, Lifetime=6.85, UsePositionOffset=false, Write Material Random=true)
   &CameraOffset(Camera Offset Amount=&MergeHelper_FloatSwitch(0=20, 1=10, 2=0, 3=0, 4=0, Index=&SpawnGroup(), Num=::NewEnumerator3))
  ]
  particleUpdateScript [
   &SolveForcesAndVelocity(Clamp Velocity=false, Limit Acceleration=false)
   &ParticleState(Kill Particles When Lifetime Has Elapsed=false, Let Infinitely Lived Particles Die When Emitter Deactivates=true)
   &SkeletalMeshLocation(
    Bone Sampling Mode=::NewEnumerator4
    MaskSpawnGroup=false
    Sampled Position Offset=&MergeHelper_VectorSwitch(0=<0 0 0>, 1=<0 0 6>, 2=<0 0 6>, 3=<0 0 0>, 4=<0 0 0>, Index=&SpawnGroup(), Num=::NewEnumerator3)
    Sampled Position Offset Coordinate Space:ENiagaraSkelMeshTransforms=::NewEnumerator2
    Skeletal Mesh=$System.Skeletal Mesh Reader
   )
   /SetVariables(
    Particles.Decal Color:LinearColor=&MultiplyLinearColor(
     Linear Color=<rgba 2.6271 19.2283 45 1>
     Scale Factor=&MakeVector4FromVectorAndFloat(
      "Vector (RGB)"=&VectorFromFloat(Value=&FloatFromCurve(
        FloatCurve=DataInterface() {propertyValues {Curve {keys [
            {interpMode=RCIM_Cubic, leaveTangentWeight=0.3333, tangentMode=RCTM_User, time=0, value=0}
            {arriveTangent=-0.068, arriveTangentWeight=0.3341, interpMode=RCIM_Cubic, leaveTangent=-0.068, tangentMode=RCTM_User, time=0.45, value=1}
           ]}}}
        CurveIndex=$Emitter.Age
       ))
      "Float (Alpha)"=&Multiply_Float(
       A=1
       B=&FloatFromCurve(
        FloatCurve=DataInterface() {propertyValues {Curve {keys [
            {interpMode=RCIM_Cubic, leaveTangentWeight=0.0667, tangentMode=RCTM_User, time=0, value=0}
            {arriveTangentWeight=0.0667, interpMode=RCIM_Cubic, tangentMode=RCTM_User, time=0.2, value=1.0088}
           ]}}}
        CurveIndex=$Emitter.Age
       )
      )
     )
    )
    Particles.Decal DP:Vector4f=&MakeVector4(
     X=&FloatFromCurve(
      FloatCurve=DataInterface() {propertyValues {Curve {keys [<at 0 0 RCIM_Cubic>,<at 0.3 1 RCIM_Cubic>]}}}
      CurveIndex=$Emitter.Age
      Scale Curve=2
     )
     Y=0.01
     Z=&FloatFromCurve(FloatCurve=DataInterface() {propertyValues {Curve {keys [<at 0 0.05>,<at 1 0>]}}})
     W=&RandomRangeFloat(Minimum=0, Maximum=1)
    )
    Particles.Decal Size:Vector2f=&Vector2DFromFloat(Value=&FloatFromCurve(
      FloatCurve=DataInterface() {propertyValues {Curve {keys [
          {interpMode=RCIM_Cubic, leaveTangentWeight=0.3333, tangentMode=RCTM_User, time=0, value=0}
          {arriveTangentWeight=0.3333, interpMode=RCIM_Cubic, tangentMode=RCTM_User, time=0.2, value=1.45}
         ]}}}
      Scale Curve=30
     ))
   ) {isSetParameters=true}
   /SetVariables(
    Particles.Inner Aura Color:LinearColor=&MultiplyLinearColor(
     Linear Color=<rgba 0.3103 2.9353 10 1>
     Scale Factor=&MakeVector4FromVectorAndFloat(
      "Vector (RGB)"=&VectorFromFloat(Value=&FloatFromCurve(
        FloatCurve=DataInterface() {propertyValues {Curve {keys [<at 0 0 RCIM_Cubic>,<at 0.45 1 RCIM_Cubic>]}}}
        CurveIndex=$Emitter.Age
        Scale Curve=&FNTimeOfDayScalar(Night Scale=0.75, Morning Scale=1, Evening Scale=1)
       ))
      "Float (Alpha)"=&FloatFromCurve(
       FloatCurve=DataInterface() {propertyValues {Curve {keys [<at 0 0 RCIM_Cubic>, {interpMode=RCIM_Cubic, leaveTangent=-0.0405, time=0.45, value=1}]}}}
       CurveIndex=$Emitter.Age
       Scale Curve=0.4
      )
     )
    )
    Particles.Inner Aura Size:Vector2f=&Multiply_Vector2D_ByFloat(Float=0.6, Vector2D=<35 60>)
   ) {isSetParameters=true}
   /SetVariables(
    Particles.Outer Aura Color:LinearColor=&MultiplyLinearColor(
     Linear Color=<rgba 0.3969 1.1104 2 1>
     Scale Factor=&MakeVector4FromVectorAndFloat(
      "Vector (RGB)"=&VectorFromFloat(Value=&FloatFromCurve(
        FloatCurve=DataInterface() {propertyValues {Curve {keys [<at 0 0 RCIM_Cubic>,<at 0.45 1 RCIM_Cubic>]}}}
        CurveIndex=$Emitter.Age
        Scale Curve=&FNTimeOfDayScalar(Night Scale=0.75, Morning Scale=1, Evening Scale=1)
       ))
      "Float (Alpha)"=&FloatFromCurve(
       FloatCurve=DataInterface() {propertyValues {Curve {keys [<at 0 0 RCIM_Cubic>, {interpMode=RCIM_Cubic, leaveTangent=-0.0405, time=0.45, value=1}]}}}
       CurveIndex=$Emitter.Age
      )
     )
    )
    Particles.Outer Aura Size:Vector2f=<40 60>
   ) {isSetParameters=true}
   /SetVariables(
    Particles.Glow Color=&MultiplyLinearColor(
     Linear Color=<rgba 0.0104 0.1135 0.5 0.25>
     Scale Factor=&MakeVector4FromVectorAndFloat(
      "Vector (RGB)"=&Multiply_VectorByFloat(Vector=<1 1 1>, Float=&FNTimeOfDayScalar(Night Scale=0.2, Morning Scale=0.5, Evening Scale=0.5))
      "Float (Alpha)"=&FloatFromCurve(
       FloatCurve=DataInterface() {propertyValues {Curve {keys [<at 0 0 RCIM_Cubic>, {interpMode=RCIM_Cubic, leaveTangent=-0.0405, time=0.45, value=1}]}}}
       CurveIndex=$Emitter.Age
      )
     )
    )
    Particles.Glow DP:Vector4f=<1 5 0 0>
    Particles.Glow Size=&FN_FrontEndVector2Modifier(Front End:Vector2f=<0 0>, Gameplay:Vector2f=<35 80>)
   ) {isSetParameters=true}
   &AlignSpriteToMeshOrientation(
    Mesh Orientation Relative Sprite Alignment Vector=<-1 0 0>
    Mesh Orientation Relative Sprite Facing Vector:Vector3f=<0 1 0>
    SetAlignment=true
    SetFacing=true
   )
  ]
  rendererClasses [&NiagaraSpriteRendererProperties]
  renderers [
   {rendererClass=&NiagaraSpriteRendererProperties}
   {rendererClass=&NiagaraSpriteRendererProperties, rendererIndex=1}
   {rendererClass=&NiagaraSpriteRendererProperties, rendererIndex=2}
   {rendererClass=&NiagaraSpriteRendererProperties, rendererIndex=3}
  ]
 }
}
summary {
 systemName=NS_Emote_FoundKelp_Intro
 emitters [
  {emitterName=ElectricCoils, rendererClasses [&NiagaraMeshRendererProperties]}
  {emitterName=GPU_particles001, rendererClasses [&NiagaraSpriteRendererProperties]}
  {emitterName=Hand_Glow, rendererClasses [&NiagaraSpriteRendererProperties]}
 ]
 userVariables [
  User.bIsFrontEnd {type=NiagaraBool}
  User.bIsFrontEndPreview {type=NiagaraBool}
  User.BlinkIndex {type=NiagaraFloat, defaultValue=0, description=""}
  User.IsFrontEnd {type=NiagaraBool}
 ]
}
